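# Cascade-converted Niagara system adding pink electric effects to the Street Racer Drift Remix backpack cosmetic.
# Three electric emitters spawn short-lived sprites from cylinders in randomized bursts, animating material parameters; two also render fading lights.
# A periodic burst emitter adds lightning meshes, while persistent and flickering sprite glows sit camera-offset.
# The low-alpha glow samples the skeletal mesh's "pack" bone; other emitters use fixed local offsets.
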NiagaraSystem /BRCosmetics/Accessories/FORT_Backpacks/Backpack_F_MED_Street_Racer_Drift_Remix/FX/NS_Backpack_StreetRacer_DriftRemix_Converted
dependencies {
 usedDataInterfaces [&NiagaraDataInterfaceCurve, &NiagaraDataInterfaceSkeletalMesh]
 usedDynamicInputs []
 usedModules [&SystemState]
}
emitters {
 Electric_Connection {
  emitterUpdateScript [
   &EmitterState(
    Life Cycle Mode=::NewEnumerator1
    Loop Delay=&RandomRangeFloat(Minimum=0.4, Maximum=0.5, Evaluation Type=::NewEnumerator1)
    Loop Duration=&RandomRangeFloat(Minimum=0.25, Maximum=0.4, Evaluation Type=::NewEnumerator1)
    UseLoopDelay=true
   )
   &SpawnRate(Use Spawn Probability=false)
  ]
  particleSpawnScript [
   /SetVariables(Particles.ShapeLocationOffset:Vector3f=<0 -4 0>) {isSetParameters=true}
   /SetVariables(Particles.ShapeLocationOffsetSpace:ENiagaraCoordinateSpace=::Local) {isSetParameters=true}
   &InitializeParticle(
    Lifetime=&RandomRangeFloat(Minimum=0.2, Maximum=0.4, Evaluation Type=::NewEnumerator1)
    Sprite Rotation Angle=&RandomRangeFloat(Minimum=0, Maximum=1, Evaluation Type=::NewEnumerator1)
    Sprite Rotation Mode=::NewEnumerator1
    Sprite Size Mode=::NewEnumerator4
    UsePositionOffset=false
    Write Material Random=true
   ) {enabled=false}
   &InitializeParticle(
    Lifetime=&RandomRangeFloat(Minimum=0.2, Maximum=0.4, Evaluation Type=::NewEnumerator1)
    Sprite Rotation Angle=&RandomRangeFloat(Minimum=0, Maximum=1, Evaluation Type=::NewEnumerator1)
    Sprite Rotation Mode=::NewEnumerator1
    Sprite Size Mode=::NewEnumerator4
    UsePositionOffset=false
    Write Material Random=true
   ) {module=InitializeParticle001}
   &Color(Color=&MakeLinearColorFromVectorAndFloat(
     "Vector (RGB)"=&VectorFromFloat(Value=&RandomRangeFloat(Minimum=1, Maximum=30, Evaluation Type=::NewEnumerator1))
     "Float (Alpha)"=1
    ))
   &DynamicMaterialParameters(
    Index 0 Param 1=&RandomRangeFloat(Minimum=2.5, Maximum=3.5, Evaluation Type=::NewEnumerator1)
    Index 0 Param 3=&RandomRangeFloat(Minimum=0.2, Maximum=0.3, Evaluation Type=::NewEnumerator1)
    Module.Param0WriteEnabled=true
    Module.Param1WriteEnabled=false
    Module.Param2WriteEnabled=true
    Module.Param3WriteEnabled=true
   )
   &ShapeLocation(
    Cylinder Height=8
    Cylinder Height Midpoint=1
    Cylinder Radius=0.5
    MaskSpawnGroup=false
    Non Uniform Scale=<1 1 0.5>
    Offset=$Particles.ShapeLocationOffset
    Offset Coordinate Space=$Particles.ShapeLocationOffsetSpace
    Offset Mode=::NewEnumerator0
    Override Seed=false
    Rotation Mode=::NewEnumerator2
    Shape Primitive=::NewEnumerator1
    Surface Only=false
    UseOwnerScale=true
    Yaw / Pitch / Roll=<0 0 0.25>
   )
   /CascadeToNiagaraConverter/NiagaraScripts/ModuleScripts/FixShapeVector(
    Offset:Vector3f=$Particles.ShapeLocationOffset
    Offset Coordinate Space:ENiagaraCoordinateSpace=$Particles.ShapeLocationOffsetSpace
    Offset Mode:ENiagara_OffsetMode=::NewEnumerator0
   )
   &AddVelocity(Velocity=&Multiply_VectorByFloat(Vector=&Multiply_Vector(A=$Output.FixShapeVector.ShapeVector, B=<1 0 1>), Float=9))
   &AddVelocity(Velocity=<0 5 0>) {module=AddVelocity001}
  ]
  particleUpdateScript [
   /SetVariables(Particles.LightExponent=2) {isSetParameters=true}
   &ParticleState()
   &ScaleColor(
    Scale Alpha=&FloatFromCurve(FloatCurve=DataInterface() {propertyValues {Curve {keys [
         {arriveTangent=-1.9446, time=-0.0016, value=3}
         {arriveTangent=-2.9954, interpMode=RCIM_Cubic, leaveTangent=-2.9954, time=0.2015, value=1.0554}
         <at 1 0>
        ]}}})
    Scale RGB=<1 0.0205 0.5567>
    ScaleA=true
    ScaleRGB=true
   )
   &DynamicMaterialParameters(
    Index 0 Param 2=&FloatFromCurve(FloatCurve=DataInterface() {propertyValues {Curve {keys [<at 0 0.5>,<at 1 -1>]}}})
    Module.Param0WriteEnabled=false
    Module.Param1WriteEnabled=true
    Module.Param2WriteEnabled=false
    Module.Param3WriteEnabled=false
   ) {module=DynamicMaterialParameters001}
   &CascadeConversion_LightProperties(
    LightBrightnessScale:NiagaraFloat=&FloatFromCurve(FloatCurve=DataInterface() {propertyValues {Curve {keys [{arriveTangent=-0.7246, time=0, value=1}, {arriveTangent=-1, interpMode=RCIM_Cubic, leaveTangent=-1, time=0.1375, value=0.2754}, <at 1 0>]}}})
    LightColorScale=<0.5 0.5 0.5>
    LightRadiusScale=0.5
   )
   &SolveForcesAndVelocity(Clamp Velocity=false, Limit Acceleration=false)
  ]
  rendererClasses [&NiagaraSpriteRendererProperties, &NiagaraLightRendererProperties]
  renderers [
   {rendererClass=&NiagaraSpriteRendererProperties}
   {rendererClass=&NiagaraLightRendererProperties, rendererIndex=1}
  ]
 }
 Electric_Connection001 {
  emitterUpdateScript [
   &EmitterState(
    Life Cycle Mode=::NewEnumerator1
    Loop Delay=&RandomRangeFloat(Minimum=0.4, Maximum=0.5, Evaluation Type=::NewEnumerator1)
    Loop Duration=&RandomRangeFloat(Minimum=0.25, Maximum=0.4, Evaluation Type=::NewEnumerator1)
    UseLoopDelay=true
   )
   &SpawnRate(Use Spawn Probability=false)
  ]
  particleSpawnScript [
   /SetVariables(Particles.ShapeLocationOffset:Vector3f=<0 -4 0>) {isSetParameters=true}
   /SetVariables(Particles.ShapeLocationOffsetSpace:ENiagaraCoordinateSpace=::Local) {isSetParameters=true}
   &InitializeParticle(
    Lifetime=&RandomRangeFloat(Minimum=0.2, Maximum=0.4, Evaluation Type=::NewEnumerator1)
    Sprite Rotation Angle=&RandomRangeFloat(Minimum=0, Maximum=1, Evaluation Type=::NewEnumerator1)
    Sprite Rotation Mode=::NewEnumerator1
    Sprite Size Mode=::NewEnumerator4
    UsePositionOffset=false
    Write Material Random=true
   ) {enabled=false}
   &InitializeParticle(
    Lifetime=&RandomRangeFloat(Minimum=0.2, Maximum=0.4, Evaluation Type=::NewEnumerator1)
    Sprite Rotation Angle=&RandomRangeFloat(Minimum=0, Maximum=1, Evaluation Type=::NewEnumerator1)
    Sprite Rotation Mode=::NewEnumerator1
    Sprite Size Mode=::NewEnumerator4
    UsePositionOffset=false
    Write Material Random=true
   ) {module=InitializeParticle001}
   &Color(Color=&MakeLinearColorFromVectorAndFloat(
     "Vector (RGB)"=&VectorFromFloat(Value=&RandomRangeFloat(Minimum=1, Maximum=30, Evaluation Type=::NewEnumerator1))
     "Float (Alpha)"=&RandomRangeFloat(Minimum=0.3, Maximum=0.6, Evaluation Type=::NewEnumerator1)
    ))
   &DynamicMaterialParameters(
    Index 0 Param 1=&RandomRangeFloat(Minimum=2.5, Maximum=3.5, Evaluation Type=::NewEnumerator1)
    Index 0 Param 3=&RandomRangeFloat(Minimum=0.2, Maximum=0.3, Evaluation Type=::NewEnumerator1)
    Module.Param0WriteEnabled=true
    Module.Param1WriteEnabled=false
    Module.Param2WriteEnabled=true
    Module.Param3WriteEnabled=true
   )
   &ShapeLocation(
    Cylinder Height=8
    Cylinder Radius=0.5
    MaskSpawnGroup=false
    Offset=$Particles.ShapeLocationOffset
    Offset Coordinate Space=$Particles.ShapeLocationOffsetSpace
    Offset Mode=::NewEnumerator0
    Override Seed=false
    Rotation Mode=::NewEnumerator2
    Shape Primitive=::NewEnumerator1
    Surface Only=false
    UseOwnerScale=true
    Yaw / Pitch / Roll=<0 0 0.25>
   )
   &FixShapeVector(
    Offset:Vector3f=$Particles.ShapeLocationOffset
    Offset Coordinate Space:ENiagaraCoordinateSpace=$Particles.ShapeLocationOffsetSpace
    Offset Mode:ENiagara_OffsetMode=::NewEnumerator0
   )
   &AddVelocity(Velocity=&Multiply_VectorByFloat(Vector=&Multiply_Vector(A=$Output.FixShapeVector.ShapeVector, B=<1 0 1>), Float=9))
  ]
  particleUpdateScript [
   &ParticleState()
   &ScaleColor(
    Scale Alpha=&FloatFromCurve(FloatCurve=DataInterface() {propertyValues {Curve {keys [
         {arriveTangent=-1.9446, time=-0.0016, value=3}
         {arriveTangent=-2.9954, interpMode=RCIM_Cubic, leaveTangent=-2.9954, time=0.2015, value=1.0554}
         <at 1 0>
        ]}}})
    Scale RGB=<1 0.0205 0.5567>
    ScaleA=true
    ScaleRGB=true
   )
   &DynamicMaterialParameters(
    Index 0 Param 2=&FloatFromCurve(FloatCurve=DataInterface() {propertyValues {Curve {keys [<at 0 0.5>,<at 1 -1>]}}})
    Module.Param0WriteEnabled=false
    Module.Param1WriteEnabled=true
    Module.Param2WriteEnabled=false
    Module.Param3WriteEnabled=false
   ) {module=DynamicMaterialParameters001}
   &SolveForcesAndVelocity(Clamp Velocity=false, Limit Acceleration=false)
  ]
  rendererClasses [&NiagaraSpriteRendererProperties]
  renderers [{rendererClass=&NiagaraSpriteRendererProperties}]
 }
 Electric_High {
  emitterUpdateScript [
   &EmitterState(
    Life Cycle Mode=::NewEnumerator1
    Loop Delay=&RandomRangeFloat(Minimum=2, Maximum=2.5, Evaluation Type=::NewEnumerator1)
    Loop Duration=&RandomRangeFloat(Minimum=0.3, Maximum=0.4, Evaluation Type=::NewEnumerator1)
    UseLoopDelay=true
   )
   &SpawnRate(SpawnRate=2, Use Spawn Probability=false)
  ]
  particleSpawnScript [
   /SetVariables(Particles.ShapeLocationOffset:Vector3f=<0 -2 0>) {isSetParameters=true}
   /SetVariables(Particles.ShapeLocationOffsetSpace:ENiagaraCoordinateSpace=::Local) {isSetParameters=true}
   &InitializeParticle(
    Lifetime=&RandomRangeFloat(Minimum=0.13, Maximum=0.22, Evaluation Type=::NewEnumerator1)
    Sprite Rotation Angle=&RandomRangeFloat(Minimum=0, Maximum=1, Evaluation Type=::NewEnumerator1)
    Sprite Rotation Mode=::NewEnumerator1
    Sprite Size Mode=::NewEnumerator4
    UsePositionOffset=false
    Write Material Random=true
   ) {enabled=false}
   &InitializeParticle(
    Lifetime=&RandomRangeFloat(Minimum=0.13, Maximum=0.22, Evaluation Type=::NewEnumerator1)
    Sprite Rotation Angle=&RandomRangeFloat(Minimum=0, Maximum=1, Evaluation Type=::NewEnumerator1)
    Sprite Rotation Mode=::NewEnumerator1
    Sprite Size Mode=::NewEnumerator4
    UsePositionOffset=false
    Write Material Random=true
   ) {module=InitializeParticle001}
   &Color(Color=&MakeLinearColorFromVectorAndFloat(
     "Vector (RGB)"=&VectorFromFloat(Value=&RandomRangeFloat(Minimum=4, Maximum=100, Evaluation Type=::NewEnumerator1))
     "Float (Alpha)"=1
    ))
   &DynamicMaterialParameters(
    Index 0 Param 1=&RandomRangeFloat(Minimum=2.5, Maximum=3.5, Evaluation Type=::NewEnumerator1)
    Index 0 Param 3=&RandomRangeFloat(Minimum=0.2, Maximum=0.3, Evaluation Type=::NewEnumerator1)
    Module.Param0WriteEnabled=true
    Module.Param1WriteEnabled=false
    Module.Param2WriteEnabled=true
    Module.Param3WriteEnabled=true
   )
   &ShapeLocation(
    Cylinder Height=5
    Cylinder Height Midpoint=1
    Cylinder Radius=9
    MaskSpawnGroup=false
    Non Uniform Scale=<1 1 0.5>
    Offset=$Particles.ShapeLocationOffset
    Offset Coordinate Space=$Particles.ShapeLocationOffsetSpace
    Offset Mode=::NewEnumerator0
    Override Seed=false
    Rotation Mode=::NewEnumerator2
    Shape Primitive=::NewEnumerator1
    Surface Only=false
    UseOwnerScale=true
    Yaw / Pitch / Roll=<0 0 0.25>
   )
   &FixShapeVector(
    Offset:Vector3f=$Particles.ShapeLocationOffset
    Offset Coordinate Space:ENiagaraCoordinateSpace=$Particles.ShapeLocationOffsetSpace
    Offset Mode:ENiagara_OffsetMode=::NewEnumerator0
   )
   &AddVelocity(Velocity=&Multiply_VectorByFloat(Vector=&Multiply_Vector(A=$Output.FixShapeVector.ShapeVector, B=<1 0 1>), Float=10))
   &AddVelocity(Velocity=<0 25 0>) {module=AddVelocity001}
  ]
  particleUpdateScript [
   /SetVariables(Particles.LightExponent=2) {isSetParameters=true}
   &ParticleState()
   &ScaleSpriteSize(
    Scale Factor=&Vector2DFromFloat(Value=&FloatFromCurve(FloatCurve=DataInterface() {propertyValues {Curve {keys [{arriveTangent=2.4335, interpMode=RCIM_Cubic, leaveTangent=2.4335, tangentMode=RCTM_User, time=0, value=0.3}, <at 1 1>]}}}))
    Scale Sprite Size Mode=::NewEnumerator2
   )
   &ScaleColor(
    Scale Alpha=&FloatFromCurve(FloatCurve=DataInterface() {propertyValues {Curve {keys [
         {arriveTangent=-1.9446, time=-0.0016, value=3}
         {arriveTangent=-2.9954, interpMode=RCIM_Cubic, leaveTangent=-2.9954, time=0.2015, value=1.0554}
         <at 1 0>
        ]}}})
    Scale RGB=<1 0.0205 0.5567>
    ScaleA=true
    ScaleRGB=true
   )
   &DynamicMaterialParameters(
    Index 0 Param 2=&FloatFromCurve(FloatCurve=DataInterface() {propertyValues {Curve {keys [<at 0 0.5>,<at 1 -1>]}}})
    Module.Param0WriteEnabled=false
    Module.Param1WriteEnabled=true
    Module.Param2WriteEnabled=false
    Module.Param3WriteEnabled=false
   ) {module=DynamicMaterialParameters001}
   &CascadeConversion_LightProperties(
    LightBrightnessScale:NiagaraFloat=&FloatFromCurve(FloatCurve=DataInterface() {propertyValues {Curve {keys [{arriveTangent=-0.7246, time=0, value=1}, {arriveTangent=-1, interpMode=RCIM_Cubic, leaveTangent=-1, time=0.1375, value=0.2754}, <at 1 0>]}}})
    LightColorScale=<1 1 1>
    LightRadiusScale=1
   )
   &SolveForcesAndVelocity(Clamp Velocity=false, Limit Acceleration=false)
  ]
  rendererClasses [&NiagaraSpriteRendererProperties, &NiagaraLightRendererProperties]
  renderers [
   {rendererClass=&NiagaraSpriteRendererProperties}
   {rendererClass=&NiagaraLightRendererProperties, rendererIndex=1}
  ]
 }
 LightningMesh {
  emitterUpdateScript [
   &EmitterState(Life Cycle Mode=::NewEnumerator1, Loop Duration=2, UseLoopDelay=true)
   &SpawnBurst_Instantaneous(Use Spawn Probability=false, UseLoopCountLimit=false) {module=SpawnBurst_Instantaneous001}
  ]
  particleSpawnScript [
   &InitializeParticle(
    Lifetime=0.7
    Mesh Scale=&VectorFromFloat(Value=&RandomRangeFloat(Minimum=0.5, Maximum=0.8, Evaluation Type=::NewEnumerator1))
    Mesh Scale Mode=::NewEnumerator4
    UsePositionOffset=false
    Write Material Random=true
   )
   &Color(Color=&MakeLinearColorFromVectorAndFloat("Vector (RGB)"=<2 0 0.8076>, "Float (Alpha)"=1))
   /CascadeToNiagaraConverter/NiagaraScripts/ModuleScripts/AddVectorToPosition(Vector:Vector3f=&TransformVector(Vector=<0 -2 0>, Destination Space=::Simulation))
  ]
  particleUpdateScript [&ParticleState()]
  rendererClasses [&NiagaraMeshRendererProperties]
  renderers [{rendererClass=&NiagaraMeshRendererProperties, rendererIndex=0}]
 }
 Low_ParticleDecal_persist {
  emitterUpdateScript [
   &EmitterState(Life Cycle Mode=::NewEnumerator1, Loop Duration=1, UseLoopDelay=true)
   &SpawnRate(SpawnRate=2, Use Spawn Probability=false)
  ]
  particleSpawnScript [
   &InitializeParticle(Sprite Size=<30 30>, Sprite Size Mode=::NewEnumerator4, UsePositionOffset=false, Write Material Random=true)
   &Color(Color=&MakeLinearColorFromVectorAndFloat("Vector (RGB)"=<1 0.115 0.821>, "Float (Alpha)"=0.08))
  ]
  particleUpdateScript [
   &ParticleState()
   &ScaleColor(Scale RGB=<3 3 3>, ScaleA=true, ScaleRGB=true)
   &SkeletalMeshLocation(
    Bone / Socket Index=/CascadeToNiagaraConverter/NiagaraScripts/DynamicInputScripts/SequentialInt(Use Array Data:NiagaraBool=false, First Number:NiagaraInt32=0, Sequence:NiagaraDataInterfaceArrayInt32=?, Last Number:NiagaraInt32=0)
    Bone Sampling Mode=::NewEnumerator4
    MaskSpawnGroup=false
    Orientation Sampling=::NewEnumerator1
    Skeletal Mesh=DataInterface() {propertyValues {
      DefaultMesh=null
      ExcludeBoneName=root
      OptionalComponentClass=null
      PreviewMesh=null
      SkinningMode=SkinOnTheFly
      SoftSourceActor=null
      SourceMode=Default
      UvSetIndex=0
      WholeMeshLOD=-1
      bExcludeBone=false
      bReadDeformedGeometry=true
      bRequireCurrentFrameData=true
      ComponentTags []
      FilteredBones [pack]
      FilteredSockets []
      MeshUserParameter {parameter {name=None, typeDefHandle {registeredTypeIndex=1}, varData []}}
      SamplingRegions []
     }}
   )
   &CameraOffset(Camera Offset Amount=3)
  ]
  rendererClasses [&NiagaraSpriteRendererProperties]
  renderers [{rendererClass=&NiagaraSpriteRendererProperties}]
 }
 ParticleDecal_Flick {
  emitterUpdateScript [
   &EmitterState(
    Life Cycle Mode=::NewEnumerator1
    Loop Delay=&RandomRangeFloat(Minimum=0.4, Maximum=0.6, Evaluation Type=::NewEnumerator1)
    Loop Duration=0.2
    UseLoopDelay=true
   )
   &SpawnBurst_Instantaneous(Use Spawn Probability=false, UseLoopCountLimit=false)
  ]
  particleSpawnScript [
   &InitializeParticle(Lifetime=0.02, Sprite Size=<15 15>, Sprite Size Mode=::NewEnumerator4, UsePositionOffset=false, Write Material Random=true)
   &Color(Color=&MakeLinearColorFromVectorAndFloat(
     "Vector (RGB)"=<1 0.115 0.821>
     "Float (Alpha)"=&RandomRangeFloat(Minimum=0.8, Maximum=1.2, Evaluation Type=::NewEnumerator1)
    ))
   &AddVectorToPosition(Vector:Vector3f=&TransformVector(Vector=<0 -2 2>, Destination Space=::Simulation))
  ]
  particleUpdateScript [
   &ParticleState()
   &CameraOffset(Camera Offset Amount=5)
   &ScaleColor(Scale RGB=<3 3 3>, ScaleA=true, ScaleRGB=true)
  ]
  rendererClasses [&NiagaraSpriteRendererProperties]
  renderers [{rendererClass=&NiagaraSpriteRendererProperties}]
 }
 ParticleDecal_persist {
  emitterUpdateScript [
   &EmitterState(Life Cycle Mode=::NewEnumerator1, Loop Duration=1, UseLoopDelay=true)
   &SpawnRate(SpawnRate=2, Use Spawn Probability=false)
  ]
  particleSpawnScript [
   &InitializeParticle(Sprite Size Mode=::NewEnumerator4, UsePositionOffset=false, Write Material Random=true)
   &Color(Color=&MakeLinearColorFromVectorAndFloat("Vector (RGB)"=<1 0.115 0.821>, "Float (Alpha)"=0.7))
   &AddVectorToPosition(Vector:Vector3f=&TransformVector(Vector=<0 -2 2>, Destination Space=::Simulation))
  ]
  particleUpdateScript [
   &ParticleState()
   &CameraOffset(Camera Offset Amount=5)
   &ScaleColor(Scale RGB=<3 3 3>, ScaleA=true, ScaleRGB=true)
  ]
  rendererClasses [&NiagaraSpriteRendererProperties]
  renderers [{rendererClass=&NiagaraSpriteRendererProperties}]
 }
}
summary {
 systemName=NS_Backpack_StreetRacer_DriftRemix_Converted
 emitters [
  {emitterName=Electric_High, rendererClasses [&NiagaraSpriteRendererProperties, &NiagaraLightRendererProperties]}
  {
   emitterName=Electric_Connection
   rendererClasses [&NiagaraSpriteRendererProperties, &NiagaraLightRendererProperties]
  }
  {emitterName=Electric_Connection001, rendererClasses [&NiagaraSpriteRendererProperties]}
  {emitterName=ParticleDecal_persist, rendererClasses [&NiagaraSpriteRendererProperties]}
  {emitterName=ParticleDecal_Flick, rendererClasses [&NiagaraSpriteRendererProperties]}
  {emitterName=Low_ParticleDecal_persist, rendererClasses [&NiagaraSpriteRendererProperties]}
  {emitterName=LightningMesh, rendererClasses [&NiagaraMeshRendererProperties]}
 ]
 userVariables []
}
