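# Physics asset for the Street Racer Drift Remix backpack cosmetic, named as its dynamics (Dyn) variant.
# It has two thin box bodies: a smaller one on the pack bone and a wider, taller one on root, offset to opposite sides along Y.
# A single constraint joins pack to root with both swings and twist locked, so pack cannot rotate relative to root.
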PhysicsAsset /BRCosmetics/Accessories/FORT_Backpacks/Backpack_F_MED_Street_Racer_Drift_Remix/Mesh/F_MED_Street_Racer_Drift_Remix_Pack_Dyn
bodies {
 pack [{shapeType=Box, extentX=16.9375, extentY=3.5625, extentZ=18.8125, center=<0 -0.724 0>}]
 root [{shapeType=Box, extentX=37.25, extentY=1.75, extentZ=29.375, center=<0 1.6997 0>}]
}
constraints [{
  bone1Name=pack
  bone2Name=root
  swing1LimitDegrees=45
  swing1Motion=Locked
  swing2LimitDegrees=0
  swing2Motion=Locked
  twistLimitDegrees=0
  twistMotion=Locked
 }]
